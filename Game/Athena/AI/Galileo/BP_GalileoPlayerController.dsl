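; AI controller Blueprint for the Galileo NPC bot in Athena, handling leash setup and patrol movement.
; When the behavior tree starts, it writes leash location and inner and outer radii to the Blackboard.
; Patrol start flags the pawn as patrolling and, on authority only, applies a patrol-speed GameplayEffect, caches the rotation rate and overrides it.
; Patrol stop clears the flag, removes the effect and restores the cached rate without an authority check.
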(event EventBeginPlay)

(event EventTick
  (params (DeltaSeconds :type Float%20%28single-precision%29)))

(event EventBlueprintOnBehaviorTreeStarted
  (CallFunction|SetBlackboardValues))

(event OnPatrollingStart(FortAthenaNpcGalileo)
  (bind _asbp_player_pawn_athena_galileo (Utilities|Casting|CastToBP_PlayerPawn_Athena_Galileo (Variables|FortAthenaAIBotController|GetPlayerBotPawn)))
  (Variables|/Game/Athena/AI/Galileo/BP_PlayerPawn_Athena_Galileo.BP_PlayerPawn_Athena_Galileo_C|SetIsPatrolling true _asbp_player_pawn_athena_galileo)
  (Networking|SwitchHasAuthority
    (:Authority
      (bind _returnvalue (GameplayEffects|ApplyGameplayEffectToTarget :self (Ability|GetAbilitySystemComponent _asbp_player_pawn_athena_galileo) :GameplayEffectClass "/Game/Athena/AI/Galileo/GE_Galileo_PatrolSpeed.GE_Galileo_PatrolSpeed_C" :Target (Ability|GetAbilitySystemComponent _asbp_player_pawn_athena_galileo) :Level 0.0))
      (Variables|Default|SetGE_Patrol_Handle _returnvalue)
      (Variables|Default|SetOriginalRotationRate (Variables|/Script/Engine.CharacterMovementComponent|GetRotationRate (Variables|/Script/Engine.Character|GetCharacterMovement _asbp_player_pawn_athena_galileo)))
      (Variables|/Script/Engine.CharacterMovementComponent|SetRotationRate "0,150.000000, 0" (Variables|/Script/Engine.Character|GetCharacterMovement _asbp_player_pawn_athena_galileo)))))

(event OnPatrollingStop(FortAthenaNpcGalileo)
  (bind _asbp_player_pawn_athena_galileo (Utilities|Casting|CastToBP_PlayerPawn_Athena_Galileo (Variables|FortAthenaAIBotController|GetPlayerBotPawn)))
  (Variables|/Game/Athena/AI/Galileo/BP_PlayerPawn_Athena_Galileo.BP_PlayerPawn_Athena_Galileo_C|SetIsPatrolling false _asbp_player_pawn_athena_galileo)
  (GameplayEffects|RemoveActiveGameplayEffect (Ability|GetAbilitySystemComponent _asbp_player_pawn_athena_galileo) (Variables|Default|GetGE_Patrol_Handle) -1)
  (Variables|/Script/Engine.CharacterMovementComponent|SetRotationRate (Variables|Default|GetOriginalRotationRate) (Variables|/Script/Engine.Character|GetCharacterMovement _asbp_player_pawn_athena_galileo)))

(fn SetBlackboardValues
  (bind _self self)
  (bind _returnvalue (AI|GetBlackboard _self))
  (AI|Components|Blackboard|SetValueasVector _returnvalue (Variables|Default|GetBB_LeashLocationKey) (Variables|Default|GetLeashLocation))
  (AI|Components|Blackboard|SetValueasFloat _returnvalue (Variables|Default|GetBB_LeashInnerRadiusKey) (Variables|Default|GetLeashInnerRadius))
  (AI|Components|Blackboard|SetValueasFloat _returnvalue (Variables|Default|GetBB_LeashOuterRadiusKey) (Variables|Default|GetLeashOuterRadius)))

(fn ConstructionScript)
