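; BP_PlayerPawn_Athena_Galileo is an Athena player pawn Blueprint for Galileo AI characters with patrol animation support.
; When character customization completes, it links the GalileoPatrolLayerAnimBP anim class layers onto the mesh.
; The IsPatrolling replication callback passes that flag to the Fort player anim instance.
; It also returns a fixed additive wall-jump montage as its fullbody hit reaction, while BeginPlay, overlap and Tick do nothing.
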(event EventBeginPlay
  (|CallParentFunction) ; This node is disabled and will not be called.\nDrag off pins to build functionality.
)

(event Collision|EventActorBeginOverlap
  (params (OtherActor :type Actor%20Object%20Reference)))

(event EventTick
  (params (DeltaSeconds :type Float%20%28single-precision%29)))

(event EventOnCharacterCustomizationCompleted
  (params (Pawn :type Fort%20Player%20Pawn%20Object%20Reference))
  (Components|SkeletalMesh|AnimationBlueprintLinking|LinkAnimClassLayers (Variables|Character|GetMesh) "/Game/Characters/Player/Male/Male_Avg_Base/LayerAnimBP/GalileoPatrolLayerAnimBP.GalileoPatrolLayerAnimBP_C"))

(fn GetCurrentFullbodyHitReactionMontage
  (return "/Game/Animation/Game/MainPlayer/Locomotion/Medium/Male/Jump/NoWep/JumpWall_NoWep_CMM_Additive_Montage.JumpWall_NoWep_CMM_Additive_Montage"))

(fn OnRep_IsPatrolling
  (bind _asfort_player_anim_instance (Utilities|Casting|CastToFortPlayerAnimInstance (Components|SkeletalMesh|GetAnimInstance (Variables|Character|GetMesh))))
  (AI|SetIsPatrolling _asfort_player_anim_instance (Variables|Default|GetIsPatrolling)))

(fn ConstructionScript
  (|CallParentFunction))
